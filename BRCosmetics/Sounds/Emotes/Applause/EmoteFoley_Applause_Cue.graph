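# Sound cue for the foley of the Applause emote in Battle Royale cosmetics.
# A mixer plays a one-shot start wave alongside a looping applause wave, whose Enveloper fades it in from silence over about half a second to layer under the start.
# Shared emote attenuation, sound class, property sheet and concurrency set govern playback, and a submix send targets music audio analysis.
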SoundCue /BRCosmetics/Sounds/Emotes/Applause/EmoteFoley_Applause_Cue
root=Mixer_0
nodes [
 {
  id=Enveloper_1
  children [WavePlayer_1]
  properties {pitchCurve {editorCurveData {keys [{time=0}]}}, volumeCurve {editorCurveData {keys [<at 0 0>,<at 0.52 1>]}}}
 }
 {id=Mixer_0, children [WavePlayer_2,Enveloper_1], properties {inputVolume [1,1]}}
 {id=WavePlayer_1, properties {bLooping=true, soundWaveAssetPtr=&/Emote_Applause_Loop}}
 {id=WavePlayer_2, properties {soundWaveAssetPtr=&/Emote_Applause_Start}}
]
properties {
 attenuationSettings=&Fort_Emotes_Athena_Att
 audioPropertiesSheet=/BRCosmetics/Sounds/Settings/PropertySheets/Emotes/APS_Cosmetic_Emote_Foley
 maxDistance=1600
 soundClassObject=/Game/Sounds/FortSoundClassesAndModes/Fort_Foley_Emote
 volumeMultiplier=1
 concurrencyOverrides {maxCount=3}
 concurrencySet [/Game/Sounds/Concurrency_Rules/Athena_EmoteFoley_Conc]
 soundSubmixSends [{minSendLevel=0, soundSubmix=&Fortnite_AudioAnalysis_Music}]
}
